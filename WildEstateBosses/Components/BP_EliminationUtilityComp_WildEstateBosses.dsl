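; Wild Estate boss elimination utility component that places the death VFX beside the boss when it dies.
; On clients, it offsets perpendicular to the player-to-boss direction for each local player and projects that point onto navigation.
; Camera line traces test that side, then the opposite, falling back to the first if both are blocked, and set a flip flag.
; It then calls parent death handling and plays Nelson's laugh at the chosen spot.
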(event EventOnComponentOwnerDied
  (params (DamagedActor :type Actor%20Object%20Reference) (Damage :type Float%20%28single-precision%29) (InstigatedBy :type Controller%20Object%20Reference) (DamageCauser :type Actor%20Object%20Reference) (HitLocation :type Vector :default "0, 0, 0") (FHitComponent :type Primitive%20Component%20Object%20Reference) (BoneName :type Name :default "None") (Momentum :type Vector :default "0, 0, 0"))
  (Utilities|FlowControl|Sequence
    (:then_0
      (Utilities|SwitchHasAuthority
        (:Remote
          (Development|PrintString "@@@@@@ OnComponentOwnerDied [BP_EliminationUtilityComp_WildEstateBosses]: CALLED - CLIENT!" false (Variables|Debug|GetbDebugComponent))
          (for _array_element (Fortnite|GetLocalFortPlayerControllers true true)
            (Utilities|IsValid (Game|Player|GetPlayerPawn _array_element)
              (:Is%20Valid
                (Variables|Default|SetPlayerPawn (Game|Player|GetPlayerPawn _array_element))
                (CallFunction|CalculateLocalVariables)
                (Default|FindVisibileSpawnLocation)
                (if (Variables|Debug|GetbDebugComponent)
                  (CallFunction|PrintDebugLogs))
                (Variables|SetDeathVFX_Transform_Location (Math|Transform|MakeTransform (Variables|Default|GetNiagaraVFXSpawnLocation))))
              (:Is%20Not%20Valid
                ;[1 0 0 0.5] Abort if there is no player pawn.
                (Development|PrintString "[OnComponentOwnerDied] Could not find PlayerPawn.  Aborting elimination VFX." true true "(R=1.000000,G=0.000000,B=0.000000,A=1.000000)" 8.0)))))))
    (:then_1
      (|CallParentFunction DamagedActor Damage InstigatedBy DamageCauser HitLocation FHitComponent BoneName Momentum)
      (Audio|PlaySoundatLocation "/WildEstateBosses/Assets/VO/IP_SC_VO_NelsonLaugh_Cue.IP_SC_VO_NelsonLaugh_Cue" (Variables|Default|GetNiagaraVFXSpawnLocation)))))

(fn CalculateLocalVariables
  (Variables|Default|SetPlayerCameraLocation (Camera|GetCameraLocation (Variables|/Script/Engine.PlayerController|GetPlayerCameraManager (Fortnite|GetFortPlayerControllerfromActor (Variables|Default|GetPlayerPawn)))))
  (Variables|Internal|SetVector_OffsetOrigin (Pawn|GetNavAgentLocation (Variables|Internal|GetAIPawn)))
  (Variables|Internal|SetVector_ToBoss (- (Pawn|GetNavAgentLocation (Variables|Internal|GetAIPawn)) (CallFunction|GetPlayerNavAgentLocation)))
  (Variables|Internal|SetVector_BossUp (Transformation|GetActorUpVector (Variables|Internal|GetAIPawn)))
  (CallFunction|CalculatePerpendicularVector))

(fn CalculatePerpendicularVector
  (Variables|Internal|SetVector_Perpendicular (Math|Vector|Normalize (Math|Vector|CrossProduct (Variables|Internal|GetVector_ToBoss) (Variables|Internal|GetVector_BossUp))))
  (return))

(fn CalculateSpawnLocation
  (bind (_returnvalue_x _returnvalue_y _returnvalue_z) (Math|Vector|BreakVector (* (Variables|Internal|GetVector_Perpendicular) (ScalableFloat|GetValueatLevel (Variables|Internal|GetSF_SideOffsetDist)))))
  (bind (_vector_offsetorigin_x _vector_offsetorigin_y _vector_offsetorigin_z) (Math|Vector|BreakVector (Variables|GetVector_OffsetOrigin_X)))
  (return (AI|Navigation|ProjectPointtoNavigation (Utilities|Operators|Add (Math|Vector|MakeVector _vector_offsetorigin_x _vector_offsetorigin_y _vector_offsetorigin_z) (Math|Vector|MakeVector _returnvalue_x _returnvalue_y _returnvalue_z)) 0 "/Game/Athena/AI/NavFilters/NavFilter_NPC_AnyTraversable.NavFilter_NPC_AnyTraversable_C")))

(fn FindVisibileSpawnLocation
  (bind _projectedlocation (Default|CalculateSpawnLocation))
  (Utilities|FlowControl|Sequence
    (:then_0
      ;[1 1 1] First check if the spawn location to the right of the boss if visible.
      (bind _output_get (LocalVariables|SetLOriginalSpawnLocation (Default|CalculateSpawnLocation)))
      (bind (_outhit _returnvalue) (Collision|LineTraceByChannel (Variables|Default|GetPlayerCameraLocation) _output_get "TraceTypeQuery1" false (Utilities|Array|MakeArray (Variables|Internal|GetAIPawn)) "ForDuration" true "(R=1.000000,G=0.000000,B=0.000000,A=1.000000)" "(R=0.000000,G=1.000000,B=0.000000,A=1.000000)" 15.0))
      (if (not _returnvalue)
        (Variables|Default|SetNiagaraVFXSpawnLocation (LocalVariables|GetLOriginalSpawnLocation))
        (Variables|Default|SetbShouldFlip true)
        (return)
        (else
          (Development|PrintString "[FindVisibleSpawnLocation] Initial visibility check failed.  Flipping and trying again." true true "(R=1.000000,G=1.000000,B=0.000000,A=1.000000)" 5.0))))
    (:then_1
      ;[1 1 1] If failed, flip the spawn location to the right and try again
      (Variables|Internal|SetVector_Perpendicular (* (Variables|Internal|GetVector_Perpendicular) -1))
      (bind (_outhit_2 _returnvalue_2) (Collision|LineTraceByChannel (Variables|Default|GetPlayerCameraLocation) _projectedlocation "TraceTypeQuery1" false (Utilities|Array|MakeArray (Variables|Internal|GetAIPawn)) "ForDuration" true "(R=1.000000,G=0.000000,B=0.000000,A=1.000000)" "(R=0.000000,G=1.000000,B=0.000000,A=1.000000)" 15.0))
      (if (not _returnvalue_2)
        (Variables|Default|SetNiagaraVFXSpawnLocation _projectedlocation)
        (Variables|Default|SetbShouldFlip false)
        (return)
        (else
          ;[1 1 1] If both checks failed, return the original spawn location to the right of the boss.
          (Development|PrintString "[FindVisibleSpawnLocation] Flipped visibility check failed.  Reverting to original location." true true "(R=1.000000,G=1.000000,B=0.000000,A=1.000000)" 5.0)
          (Variables|Default|SetNiagaraVFXSpawnLocation (LocalVariables|GetLOriginalSpawnLocation))
          (Variables|Default|SetbShouldFlip true)
          (return))))))

(fn GetPlayerNavAgentLocation
  (bind _returnvalue (Fortnite|GetFirstLocalFortPlayerController true))
  (bind _returnvalue_2 (Game|Player|GetPlayerPawn _returnvalue))
  (Utilities|IsValid _returnvalue_2
    (:Is%20Valid
      (return (Pawn|GetNavAgentLocation _returnvalue_2)))
    (:Is%20Not%20Valid
      ;[1 1 1] If there's no player pawn, use the boss' location instead.
      (return (Pawn|GetNavAgentLocation (Variables|Internal|GetAIPawn))))))

(fn PrintDebugLogs
  (Rendering|Debug|DrawDebugSphere (Variables|Internal|GetVector_OffsetOrigin) 5e+01 12 "(R=0.005771,G=0.489583,B=0.340933,A=1.000000)" 1e+01 3.0)
  (Rendering|Debug|DrawDebugSphere (CallFunction|GetPlayerNavAgentLocation) 5e+01 12 "(R=0.075093,G=0.953125,B=0.062852,A=1.000000)" 1e+01 3.0)
  (Rendering|Debug|DrawDebugSphere (Variables|Default|GetNiagaraVFXSpawnLocation) 5e+01 12 "(R=1.000000,G=0.000000,B=0.949739,A=1.000000)" 1e+01 2.0))
